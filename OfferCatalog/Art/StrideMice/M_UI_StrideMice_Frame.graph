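# Translucent UI material that draws a rounded, outlined frame with gradient color, built on Item Shop tile functions for StrideMice.
# Three colors from the shop background-params function feed angled linear gradients, with a hard-light blend tinting the outline gradient.
# A rounded-mask function, using the shared MPC_ItemShop corner radius, supplies the background shape and outline.
# The outline mask adds a faint soft falloff derived from its SDF.
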Material /OfferCatalog/Art/StrideMice/M_UI_StrideMice_Frame
graph {
 properties {
  bAutomaticallySetUsageInEditor=false
  bUsedWithStaticMesh=false
  blendMode=BLEND_Translucent
  materialDomain=MD_UI
  usedShadingModels=Default Lit
 }
 nodes [
  LinearInterpolate(A=NamedRerouteUsage(), B=Divide(A=NamedRerouteUsage()) {constB=1}, Alpha=Multiply(A=NamedRerouteUsage()))
  MaterialFunctionCall!(Input Alpha=Constant() {r=1}, bOverrideCornerRadius, Corner Radius Override) {materialFunction=/FortUILibrary/Systems/ItemShopTile/Materials/MaterialFunctions/MF_UI_ItemShop_RoundedCorners}
  NamedRerouteDeclaration(None=MaterialFunctionCall() {materialFunction=&MF_UI_UVScaling}.Scaled to Fit) {name=UVScaling_Fit, nodeColor=<rgba 0.0781 0 1 1>}
  n1 = NamedRerouteUsage()
  n2 = NamedRerouteUsage()
  n3 = NamedRerouteUsage()
  n4 = Constant()
  n5 = Constant() {r=1}
  n6 = Constant() {r=-0.25}
  NamedRerouteDeclaration(None=LinearInterpolate(
    A=NamedRerouteUsage()
    B=MaterialFunctionCall!(Base=@n1, Blend=@n2) {materialFunction=/Engine/Functions/Engine_MaterialFunctions03/Blends/Blend_HardLight}.Result
    Alpha=MaterialFunctionCall(UVs=@n3, GradientStart=@n4, GradientEnd=@n5, RotationAngle=@n6) {materialFunction=&MF_UI_LinearGradient}.Result
   )) {name=OutlineGradient, nodeColor=<rgba 0 0 0 1>}
  NamedRerouteDeclaration(None=NamedRerouteUsage()) {name=Color2, nodeColor=<rgba 1 0.0007 0.2048 1>}
  n7 = MaterialFunctionCall() {materialFunction=/FortUILibrary/Systems/ItemShopTile/Materials/MaterialFunctions/MF_UI_ItemShop_BackgroundParams}
  NamedRerouteDeclaration(None=@n7.Color 2) {name=Secondary Color, nodeColor=<rgba 0 0 0 1>}
  NamedRerouteDeclaration(None=@n7.Color 3) {name=Accent Color, nodeColor=<rgba 0 0 0 1>}
  n8 = MaterialFunctionCall(UVs=NamedRerouteUsage(), GradientStart=Constant(), GradientEnd=Constant() {r=1.2}, RotationAngle=Constant() {r=0.25}) {materialFunction=&MF_UI_LinearGradient}
  NamedRerouteDeclaration(None=LinearInterpolate(A=NamedRerouteUsage(), B=NamedRerouteUsage(), Alpha=SmoothStep(Value=@n8.Result) {constMax=1.1})) {name=Gradient, nodeColor=<rgba 0 0 0 1>}
  NamedRerouteDeclaration(None=NamedRerouteUsage()) {name=Color3, nodeColor=<rgba 0 1 0.625 1>}
  n9 = CollectionParameter() {collection=/FortUILibrary/Systems/ItemShopTile/MPC_ItemShop, parameterName=CornerRadius}
  n10 = MaterialFunctionCall(
   Corner Radius=MaterialFunctionCall(R=@n9, G=@n9, B=@n9, A=@n9) {materialFunction=&AppendMany}.RGBA
   Radius in Pixels=Constant() {r=1}
   Outline Thickness=Constant() {r=6}
  ) {materialFunction=&MF_UI_CoreUI_RoundedMask}
  n11 = OneMinus(None=SmoothStep(Value=@n10.Outline SDF) {constMax=0.1})
  NamedRerouteDeclaration(None=LinearInterpolate(A=Multiply(A=@n11) {constB=0.2}, Alpha=@n10.Outline Mask)) {name=OutlineFrame, nodeColor=<rgba 0 0.5547 1 1>}
  NamedRerouteDeclaration(None=@n10.Shape Mask) {name=FlatBGMask, nodeColor=<rgba 0 0.7188 1 1>}
  NamedRerouteDeclaration(None=@n7.Color 1) {name=Primary Color, nodeColor=<rgba 0 0 0 1>}
  NamedRerouteDeclaration(None=NamedRerouteUsage()) {name=Color1, nodeColor=<rgba 1 0.0007 0.2048 1>}
 ]
}
